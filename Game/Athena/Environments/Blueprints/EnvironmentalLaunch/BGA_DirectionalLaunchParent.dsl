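; Parent Blueprint for directional launchers that fling overlapping players and, optionally, vehicles, granting them a launch ability and playing sounds with enemy sound indicators.
; Once playlist data is ready, scalable-float toggles either remove the actor or disable its collision.
; Players can receive tag-based velocity overrides, plus a direct LaunchCharacter on physics-simulated islands.
; In editor, the construction script draws the predicted gravity-scaled trajectory on a spline.
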(event Custom|SoftDeactivation
  (Collision|SetCollisionEnabled (Variables|Default|GetSphere) "NoCollision"))

(event Custom|LauncherEnabled
  (Variables|LaunchSettings|Gameplay|SetOriginalLaunchVelocity (Variables|LaunchSettings|Gameplay|GetLaunchVelocity)))

(event EventBeginPlay
  (Class|FortAsyncActionCurrentPlaylistReady|PlaylistDataReadyAsync
    (:OnReady
      (if (Default|IsHardEnabled)
        (if (Default|IsSoftEnabled)
          (CallFunction|LauncherEnabled)
          (else
            (CallFunction|SoftDeactivation)))
        (else
          (Actor|SetLifeSpan (Math|Random|RandomFloatinRange 0.1 1.0)))))))

(event Collision|EventActorBeginOverlap
  (params (OtherActor :type Actor%20Object%20Reference))
  ;[0.1768 0.4743 1] Event ActorBeginOverlap
  ; Check overlap for actors and conditions which should trigger a launch of that object
  (Utilities|CanAcceptOverlaps
    (:true
      (bind (_fortpawn _fortvehicle) (Utilities|PlayerorVehicle OtherActor)
        (:Player
          (if (Variables|LaunchSettings|Gameplay|GetOverrideVelocityBasedOnTag)
            ;;[0.1768 0.4743 1] Handle a velocity override case - Need to define desired Override Velocity and the Tag to check
            (if (GameplayTags|HasMatchingGameplayTag _fortpawn (Variables|LaunchSettings|Gameplay|GetTagToCheckForVelocityOverride))
              (Variables|LaunchSettings|Gameplay|SetLaunchVelocity (Variables|LaunchSettings|Gameplay|GetVelocityOverride))
              (else
                (Variables|LaunchSettings|Gameplay|SetLaunchVelocity (Variables|LaunchSettings|Gameplay|GetOriginalLaunchVelocity)))))
          (if (Fortnite|IsPhysicallySimulatedIsland)
            (bind _returnvalue (CallFunction|GetLaunchTowardLaunchTargetAngle _fortpawn self (Variables|LaunchSettings|Gameplay|GetLaunchDirectionOfUserIsHorizontalOnly)))
            (bind _level (DirectionalLauncher|GetVelocityLevel(Message) self))
            (Networking|SwitchHasAuthority
              (:Authority
                (switch Utilities|FlowControl|Switch|SwitchonENetRole (Networking|GetLocalRole _fortpawn)
                  (:ROLE_SimulatedProxy
                    _)
                  (:ROLE_AutonomousProxy
                    _)
                  (:ROLE_Authority
                    _)))
              (:Remote
                (switch Utilities|FlowControl|Switch|SwitchonENetRole (Networking|GetLocalRole _fortpawn)
                  (:ROLE_AutonomousProxy
                    _))))
            ;; Only Launch AP on Client
            (Character|LaunchCharacter _fortpawn (* (Math|Vector|SelectVector _returnvalue (Math|Rotator|GetRotationXVector (Transformation|GetWorldRotation (Variables|Default|GetLaunchDirection))) (Variables|LaunchSettings|Gameplay|GetLaunchesIndirectionOfUser)) (* _level 1e+02)) true true)
            (Networking|SwitchHasAuthority
              (:Authority
                _))
            (else
              (Networking|SwitchHasAuthority
                (:Authority
                  _))))
          (CallFunction|PlayerLaunched _fortpawn))
        (:Vehicle
          (if (Variables|LaunchSettings|Editable|GetbEnableVehicleLaunch)
            (CallFunction|VehicleLaunched _fortvehicle)))))))

(event Custom|PlayerLaunched
  (params (Player :type Fort%20Pawn%20Object%20Reference))
  ;[0.0349 0.0777 1] PlayerLaunched
  ; Called when player has successfully gone through the checks to be launched - activates launch ability on pawn and sets up VFX/SFX
  (bind (_playerabilitysystemcomponent _abilitygrantedonlaunched) (Utilities|GetValidAbilitytoUseonPlayerLaunch Player)
    (:Valid
      (Ability|Advanced|GrantandActivateAbilityonSelf _playerabilitysystemcomponent _abilitygrantedonlaunched self)
      (CallFunction|PlayFXOnClient Player))))

(event Custom|PlayFXOnClient
  (params (Pawn :type Fort%20Pawn%20Object%20Reference))
  (bind _returnvalue (Transformation|GetActorLocation))
  (bind _sfx_npc_launched_local_3d (Variables|FX|Sounds|GetSFX_NPC_Launched_Local_3D%20))
  ;[0.0349 0.0777 1] PlayerLaunched
  ; Called when player has successfully gone through the checks to be launched - activates launch ability on pawn and sets up VFX/SFX
  ;;[0.2297 0 1] SFX/VFX
  ;; Play 2d sound for instigating player, 3d sound at location for other relevant clients
  (Networking|CanExecuteCosmeticEvents
    (:True
      (Utilities|IsValid Pawn
        (:Is%20Valid
          (if (HelperFunctions|IsLocalPawn Pawn)
            (Audio|PlaySound2D (Variables|FX|Sounds|GetSFX_PlayerLaunched_local_2D))
            (else
              (Audio|PlaySoundatLocation _sfx_npc_launched_local_3d _returnvalue)
              (Fortnite|ShowSoundIndicatorLocalforTeam :Location _returnvalue :MaxAudibleDistance (Variables|/Script/Engine.SoundBase|GetMaxDistance _sfx_npc_launched_local_3d) :IndicatorType "Glider" :Team (Class|FortPawn|GetTeam Pawn) :Affiliations (Utilities|Array|MakeArray "Hostile") :Tint "(R=0.000000,G=0.000000,B=0.000000,A=0.000000)" :OverrideIcon "/Game/UI/Foundation/Textures/Icons/Navigation/SV_Launcher.SV_Launcher" :AttenuationMaxRangeMultiplier 1.0 :AttenuationIntensityMultiplier 1.0))))))))

(event Custom|VehicleLaunched
  (params (Vehicle :type Fort%20Athena%20Vehicle%20Object%20Reference))
  (bind (_outvehicleabilitysystemcomponent _ability) (Utilities|GetValidAbilitytoUseonVehicleLaunch Vehicle)
    (:Valid
      (Ability|Advanced|GrantandActivateAbilityonSelf _outvehicleabilitysystemcomponent _ability self)
      (CallFunction|PlayFXForVehicle Vehicle) ; Separated like player FX is, so device versions (child BP) can override
    )))

(fn BP_GetGameplayTags
  (if (Fortnite|IsPhysicallySimulatedIsland)
    (return "(GameplayTags=((TagName=\"Creative.IsPhysicsEnabled\")))")))

(fn GetLaunchTowardLaunchTargetAngle
  (params (Target :type Actor%20Object%20Reference) (Launcher :type Actor%20Object%20Reference) (HorizontalOnly :type Boolean))
  (bind (_returnvalue_x _returnvalue_y _returnvalue_z) (Math|Vector|BreakVector (Transformation|GetActorLocation Target)))
  (bind (_returnvalue_x_2 _returnvalue_y_2 _returnvalue_z_2) (Math|Vector|BreakVector (Transformation|GetActorLocation Launcher)))
  (return (Math|Vector|Normalize (- (Math|Vector|MakeVector _returnvalue_x _returnvalue_y (Math|Float|SelectFloat 0.0 _returnvalue_z HorizontalOnly)) (Math|Vector|MakeVector _returnvalue_x_2 _returnvalue_y_2 (Math|Float|SelectFloat 0.0 _returnvalue_z_2 HorizontalOnly))))))

(fn GetVehicleCenterOfMassLocationAndMassValue
  (params (Target :type Fort%20Vehicle%20Interface%20Interface))
  (bind _returnvalue (Vehicle|GetActor Target))
  (bind _asprimitive_component (Utilities|Casting|CastToPrimitiveComponent (Variables|/Script/Engine.Actor|GetRootComponent _returnvalue))
    (:then
      (bind (_returnvalue_x _returnvalue_y _returnvalue_z) (Math|Vector|BreakVector (Physics|GetCenterOfMass _asprimitive_component "None")))
      (return _returnvalue_x _returnvalue_y (- _returnvalue_z 1.0) (Physics|GetMass _asprimitive_component)))
    (:CastFailed
      (return (Transformation|GetActorLocation _returnvalue) 4e+02))))

(fn IsHardEnabled
  (return (>= (ScalableFloat|GetValueatLevel (Variables|LaunchSettings|Default|GetHardEnabled)) 1.0)))

(fn IsPlayerDeployEnabled
  (return (>= (ScalableFloat|GetValueatLevel (Variables|LaunchSettings|Default|GetPlayerDeployEnabled)) 1.0)))

(fn IsSoftEnabled
  (return (>= (ScalableFloat|GetValueatLevel (Variables|LaunchSettings|Default|GetSoftEnabled)) 1.0)))

(fn JustPlayer
  (params (InActor :type Actor%20Object%20Reference))
  (if (not (CallFunction|PlayerinVehicleorJustVehicle InActor))
    (bind _asfort_pawn (Utilities|Casting|CastToFortPawn InActor)
      (:then
        (return _asfort_pawn true))
      (:CastFailed
        _))
    (else
      _))
  (return 0 false))

(fn PlayerInVehicleOrJustVehicle
  (params (InActor :type Actor%20Object%20Reference))
  (bind _asfort_pawn (Utilities|Casting|CastToFortPawn InActor)
    (:then
      (bind _vehicle_occupied (Utilities|IsinVehicle _asfort_pawn)
        (:true
          (bind _asfort_athena_vehicle (Utilities|Casting|CastToFortAthenaVehicle _vehicle_occupied)
            (:then
              (return true _asfort_athena_vehicle))
            (:CastFailed
              (return false 0))))
        (:false
          _)))
    (:CastFailed
      _))
  (bind _asfort_athena_vehicle_2 (Utilities|Casting|CastToFortAthenaVehicle InActor)
    (:then
      (return true _asfort_athena_vehicle_2))
    (:CastFailed
      (return false 0))))

(fn PlayFXForVehicle
  (params (Vehicle :type Fort%20Athena%20Vehicle%20Object%20Reference))
  (bind _returnvalue (Transformation|GetActorLocation))
  (bind _sfx_veh_launched_local_3d (Variables|FX|Sounds|GetSFX_Veh_Launched_Local_3D))
  (Audio|PlaySoundatLocation _sfx_veh_launched_local_3d _returnvalue)
  (Fortnite|ShowSoundIndicatorLocalforTeam :Location _returnvalue :MaxAudibleDistance (Variables|/Script/Engine.SoundBase|GetMaxDistance _sfx_veh_launched_local_3d) :IndicatorType "Vehicle" :Team (Team|GetActorTeam Vehicle) :Affiliations (Utilities|Array|MakeArray "Hostile") :Tint "(R=0.000000,G=0.000000,B=0.000000,A=0.000000)" :OverrideIcon "/Game/UI/Foundation/Textures/Icons/Navigation/SV_Launcher.SV_Launcher" :AttenuationMaxRangeMultiplier 1.0 :AttenuationIntensityMultiplier 1.0))

(fn ConstructionScript
  (if (Editor|InEditororPIE)
    (bind _returnvalue (Class|CreativeBlueprintLibrary|GetGravityZ self))
    (bind (_outhit _outpathpositions _outlasttracedestination _returnvalue_2) (Game|PredictProjectilePathByTraceChannel (Transformation|GetWorldLocation (Variables|Default|GetLaunchDirection)) (* (Math|Rotator|GetRotationXVector (Transformation|GetWorldRotation (Variables|Default|GetLaunchDirection))) (ScalableFloat|GetValueatLevel (Variables|LaunchSettings|Gameplay|GetLaunchVelocity))) true 16.0 "ECC_GameTraceChannel11" false (Utilities|Array|MakeArray self) "None" 0.0 15.0 3e+01 (* _returnvalue (ScalableFloat|GetValueatLevel (Variables|LaunchSettings|Gameplay|GetGravityScale)))))
    (if _returnvalue_2
      (Spline|SetSplinePoints (Variables|Default|GetVisualPath_EDITOR) _outpathpositions "World" true))))

(fn ValidateAbilityParentClass
  (params (TestClass :type Object%20Class%20Reference))
  (return (Utilities|ClassIsChildOf TestClass (Variables|LaunchSettings|Editable|Defined|GetDefinedParentClass))))
