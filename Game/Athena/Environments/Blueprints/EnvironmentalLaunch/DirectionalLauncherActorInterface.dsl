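; Blueprint interface for directional launcher actors in Athena's environmental launch system, with functions that return only default values here.
; Getters expose the launch actor and its type, launch angle, velocity, velocity level, and device and player FX.
; Handler functions cover ability granting with a gravity-only check option, enabled state, launch angle and speed, and in-game visibility.
; InitSplineMeshComponents adds a hook for spline mesh setup, presumably left to implementers.
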(fn GetFX_OnDevice
  (return false))

(fn GetFX_OnPlayer
  (return false))

(fn GetLaunchActor
  (return 0))

(fn GetLaunchActorType
  (return "NewEnumerator1"))

(fn GetLaunchAngle
  (return 0.0))

(fn GetLaunchVelocity
  (return "0, 0, 0" 0.0 0.0))

(fn GetVelocityLevel
  (return 0.0))

(fn HandleAbilityGranted
  (params (OnlyCheckGravity :type Boolean))
  (return false))

(fn HandleEnabledState
  (params (NewEnable :type Boolean))
  (return))

(fn HandleLaunchAngle
  (return 0.0))

(fn HandleLaunchSpeed
  (return 0.0))

(fn HandleVisibleDuringGame
  (return false))

(fn InitSplineMeshComponents
  (return 0))
