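; Parent gameplay ability that launches pawns and vehicles from a directional environmental launcher, using the source object's direction and velocity.
; Activation branches: vehicles receive an impact and gravity-adjusting effect, while pawns get LaunchCharacterJump, fall-damage immunity, feedback and an optional timed redeploy effect.
; Landing, walking and emergency timers end the ability, which removes launch and redeploy effects.
; A separate function, not called on activation, launches physics objects by setting linear velocity.
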(event Custom|ApplyGameplayCueVehicleLaunch
  ;[1 1 1] Gameplay Cues
  (Ability|ExecuteGameplayCueOnOwner (Variables|Launched|Settings|VisualFeedback|GetVehicleLaunchedGameplayCue)))

(event Custom|ApplyGameplayCuePlayerLaunch
  ;[1 1 1] Gameplay Cues
  (Ability|ExecuteGameplayCueOnOwner (Variables|Launched|Settings|VisualFeedback|GetPlayerLaunchedGameplayCue)))

(event Custom|AttemptLaunchVehicle
  (params (Vehicle :type Fort%20Athena%20Vehicle%20Object%20Reference))
  (bind _returnvalue (Ability|GetAbilitySystemComponent Vehicle))
  (bind _geappliedlaunched_vehicle (Variables|Launched|Settings|GameplayEffect|GetGEAppliedLaunched_Vehicle))
  ;[1 1 1] Launch Vehicle
  (bind (_directionallaunchparent _velocity) (Utilities|GetValidSourceObjectasDirectionalLaunchParent)
    (:Valid
      (Audio|BroadcastAbilityImpactatLocation 0 Vehicle (* (Math|Rotator|GetRotationXVector (Transformation|GetWorldRotation (Variables|/Game/Athena/Environments/Blueprints/EnvironmentalLaunch/BGA_DirectionalLaunchParent.BGA_DirectionalLaunchParent_C|GetLaunchDirection _directionallaunchparent))) (* (ScalableFloat|GetValueatLevel _velocity (CallFunction|GetBaseVelocityLevel)) (CallFunction|GetVehicleVelocityModifer))) 777.0)
      (CallFunction|ApplyGameplayCueVehicleLaunch)
      ;[0.5943 0.776 0 0.5] Used by Creative devices to offset the way vehicles have different gravity from players
      (if (and (Utilities|IsValidClass _geappliedlaunched_vehicle) (not (GameplayTags|HasMatchingGameplayTag Vehicle (Variables|Default|GetFlyingVehicleTag))))
        (bind _returnvalue_2 (Ability|GameplayEffect|AssignTagSetbyCallerMagnitude (GameplayEffects|MakeOutgoingSpec :self _returnvalue :GameplayEffectClass _geappliedlaunched_vehicle :Level 0.0) "(TagName=\"SetByCaller.Magnitude\")" (Variables|Launched|Settings|Vehicle|GetVehicleGravityMultiplier)))
        (bind _returnvalue_3 (GameplayEffects|ApplyGameplayEffectSpecToSelf _returnvalue _returnvalue_2))
        (Variables|Launched|GameplayDynamic|SetGELaunchedAppliedHandle_Vehicle _returnvalue_3)
        (CallFunction|StartTimerCheckVehicleLanded Vehicle)
        (else
          (Ability|EndAbility))))
    (:Invalid
      (Ability|EndAbility))))

(event Custom|StartTimer_CheckVehicleLanded
  (params (Vehicle :type Fort%20Athena%20Vehicle%20Object%20Reference))
  ;[1 1 1] Launch Vehicle
  ;; Timer_CheckVehicleLanded
  (Variables|Launched|GameplayDynamic|SetTimerCheck_VehicleRef Vehicle)
  (bind _returnvalue (Utilities|Time|SetTimerbyEvent (delegate Custom|CheckVehicleLanded) 0.2 true false 0.2)) ; The delay should let us avoid any possibility of accidentally ending the ability at the start of the launch
  (Variables|Launched|GameplayDynamic|SetTimerCheckVehicleLandedHandle _returnvalue)
  (bind _returnvalue_2 (Utilities|Time|SetTimerbyEvent (EventDispatchers|CreateEvent :!event Custom|EmergencyTimer_Vehicle) (ScalableFloat|GetValueatLevel (Variables|Launched|Settings|Hotfix|GetEmergencyShutdownTime_Vehicle))))
  (Variables|Launched|GameplayDynamic|SetEmergencyTimerHandle_Vehicle _returnvalue_2))

(event Custom|AttemptLaunchFortPawn
  (params (FortPawn :type Fort%20Pawn%20Object%20Reference))
  (bind _returnvalue (Ability|GetAbilitySystemComponent FortPawn))
  (bind _geappliedlaunched (Variables|Launched|Settings|GameplayEffect|GetGEAppliedLaunched))
  (bind _geappliedimmunitytofalldamagetemplate (Variables|Launched|Settings|GameplayEffect|GetGEAppliedImmunityToFallDamageTemplate))
  ;[1 1 1] Launch FortPawn
  (bind (_directionallaunchparent _velocity) (Utilities|GetValidSourceObjectasDirectionalLaunchParent)
    (:Valid
      (if (not (Fortnite|IsPhysicallySimulatedIsland))
        (Pawn|Character|LaunchCharacterJump FortPawn (* (Math|Vector|SelectVector (CallFunction|GetLaunchTowardLaunchTargetAngle FortPawn _directionallaunchparent (Variables|/Game/Athena/Environments/Blueprints/EnvironmentalLaunch/BGA_DirectionalLaunchParent.BGA_DirectionalLaunchParent_C|GetLaunchDirectionOfUserIsHorizontalOnly _directionallaunchparent)) (Math|Rotator|GetRotationXVector (Transformation|GetWorldRotation (Variables|/Game/Athena/Environments/Blueprints/EnvironmentalLaunch/BGA_DirectionalLaunchParent.BGA_DirectionalLaunchParent_C|GetLaunchDirection _directionallaunchparent))) (Variables|/Game/Athena/Environments/Blueprints/EnvironmentalLaunch/BGA_DirectionalLaunchParent.BGA_DirectionalLaunchParent_C|GetLaunchesIndirectionOfUser _directionallaunchparent)) (ScalableFloat|GetValueatLevel _velocity (CallFunction|GetBaseVelocityLevel))) (Variables|/Game/Athena/Environments/Blueprints/EnvironmentalLaunch/BGA_DirectionalLaunchParent.BGA_DirectionalLaunchParent_C|GetOverridesVel _directionallaunchparent) (Variables|/Game/Athena/Environments/Blueprints/EnvironmentalLaunch/BGA_DirectionalLaunchParent.BGA_DirectionalLaunchParent_C|GetOverridesVel _directionallaunchparent) false true))
      (CallFunction|ListenforWalking)
      (CallFunction|ApplyGameplayCuePlayerLaunch)
      ;;[1 0.1257 0] This will handle itself as the immunity grant will give an ability that listens for when to turn off the FX
      (if (Utilities|IsValidClass _geappliedimmunitytofalldamagetemplate)
        (GameplayEffects|ApplyGameplayEffectToSelf :self _returnvalue :GameplayEffectClass _geappliedimmunitytofalldamagetemplate :Level 0.0)
        (bind _asfort_aipawn (Utilities|Casting|CastToFortAIPawn FortPawn)
          (:then
            (Variables|/Script/FortniteGame.FortPawn|SetbIgnoreNextFallingDamage true _asfort_aipawn) ; Forcibly ignore next falling damage. Player-based pawns work on a tag alone\nbut FortAIPawns need this set.
          )
          (:CastFailed
            _)) ; If we are FortAIPawn (creature, Deimos, etc.)
      )
      (if (Utilities|IsValidClass _geappliedlaunched)
        (bind _returnvalue_2 (GameplayEffects|ApplyGameplayEffectToSelf :self _returnvalue :GameplayEffectClass _geappliedlaunched :Level 0.0))
        (Variables|Launched|GameplayDynamic|SetGELaunchedAppliedHandle _returnvalue_2)
        (if (Default|IsPlayerDeployEnabled _directionallaunchparent)
          (CallFunction|BindRedeployPromptNotfication)
          (CallFunction|ApplyVisualFeedback false)
          (else
            _)))
      (CallFunction|ApplyVisualFeedback true))
    (:Invalid
      (Ability|EndAbility))))

(event Custom|EmergencyTimer_Vehicle
  ;[1 0.7253 0 0.5] Clear emergency timer for a launched vehicle
  (Utilities|Time|ClearandInvalidateTimerbyHandle (Variables|Launched|GameplayDynamic|GetEmergencyTimerHandle_Vehicle))
  (Ability|EndAbility))

(event Ability|EventActivateAbility
  ;[1 1 1] On Activate
  (bind (_athenavehicle _fortpawn) (Utilities|GetValidVehicleorPawnfromInfo)
    (:Vehicle
      (CallFunction|AttemptLaunchVehicle _athenavehicle))
    (:Player
      (CallFunction|AttemptLaunchFortPawn _fortpawn))
    (:Neither
      (Ability|EndAbility))))

(event Custom|BindRedeployPromptNotfication
  ;[1 1 1] Redeploy Logic
  (bind _returnvalue (Utilities|Time|SetTimerbyEvent (delegate Custom|RedeployPrompt) (ScalableFloat|GetValueatLevel (Variables|Launched|Settings|Hotfix|GetRedeployTime))))
  (Variables|Launched|GameplayDynamic|SetRedeployPromptTimerHandle _returnvalue)
  (bind _returnvalue_2 (Utilities|Time|SetTimerbyEvent (delegate Custom|EmergencyTimer) (ScalableFloat|GetValueatLevel (Variables|Launched|Settings|Hotfix|GetEmergencyShutdownTime))))
  (Variables|Launched|GameplayDynamic|SetEmergencyTimerHandle _returnvalue_2))

(event Custom|EmergencyTimer
  ;[1 1 1] Redeploy Logic
  (Utilities|Time|ClearandInvalidateTimerbyHandle (Variables|Launched|GameplayDynamic|GetEmergencyTimerHandle))
  (CallFunction|ForceEndAbilityIgnoreLanded))

(event Custom|RedeployPrompt
  ;[1 1 1] Redeploy Logic
  (Utilities|Time|ClearandInvalidateTimerbyHandle (Variables|Launched|GameplayDynamic|GetRedeployPromptTimerHandle))
  (if (Utilities|IsValidClass (Variables|Launched|Settings|GameplayEffect|GetGEAppliedRedeployTemplate))
    (bind _fortplayerpawn (Utilities|GetValidFortPlayerPawnfromInfo)
      (:Valid
        (bind _returnvalue (GameplayEffects|ApplyGameplayEffectToSelf :self (Ability|GetAbilitySystemComponent _fortplayerpawn) :GameplayEffectClass (Variables|Launched|Settings|GameplayEffect|GetGEAppliedRedeployTemplate) :Level 0.0))
        (Variables|Launched|GameplayDynamic|SetGERedeployHandle _returnvalue)
        (EventDispatchers|AssignOnPawnLanded _fortplayerpawn (delegate Custom|OnPawnLanded_Event_0))
        (EventDispatchers|AssignOnBeginParachuteMovementEvent _fortplayerpawn (delegate Custom|OnParachute)))
      (:Invalid
        (Ability|EndAbility)))
    (else
      (Ability|EndAbility))))

(event Custom|OnParachute
  ;[1 1 1] Redeploy Logic
  (EventDispatchers|UnbindEventfromOnBeginParachuteMovementEvent (stale Custom|RedeployPrompt _fortplayerpawn) (delegate Custom|OnParachute))
  (bind _fortplayerpawn (Utilities|GetValidFortPlayerPawnfromInfo)
    (:Valid
      (GameplayEffects|RemoveActiveEffectswithGrantedTags (Ability|GetAbilitySystemComponent _fortplayerpawn) (Variables|Launched|Settings|GameplayEffect|GetGTDefaultImmunityGrantedTag))
      (GameplayEffects|RemoveActiveGameplayEffect (Ability|GetAbilitySystemComponent _fortplayerpawn) (Variables|Launched|GameplayDynamic|GetGERedeployHandle) -1))
    (:Invalid
      (bind _wildlife_actor (Utilities|IsActorWildlife)
        (:Valid
          ;;[1 0 0.649] This tag is default granted through the generic Immune to fall damage GE / Ability. if you change that ability this tag needs to be updated in the subclass
          (GameplayEffects|RemoveActiveEffectswithGrantedTags (Ability|GetAbilitySystemComponent _wildlife_actor) (Variables|Launched|Settings|GameplayEffect|GetGTDefaultImmunityGrantedTag)))))))

(event Custom|OnPawnLanded_Event_0
  (params (Hit :type Hit%20Result%20Structure%20%28by%20ref%29))
  (bind (_period _bexecuteperiodiceffectonapplication _periodicinhibitionpolicy _modifiers _executions _conditionalgameplayeffects _overfloweffects _bdenyoverflowapplication _bclearstackonoverflow _prematureexpirationeffectclasses _routineexpirationeffectclasses _brequiremodifiersuccesstotriggercues _gameplaycues _inheritablegameplayeffecttags _inheritableownedtagscontainer) (Class|GetClassDefaults (Variables|Launched|Settings|GameplayEffect|GetGEAppliedRedeployTemplate)))
  ;[1 1 1] Redeploy Logic
  (bind _fortplayerpawn (Utilities|GetValidFortPlayerPawnfromInfo)
    (:Valid
      (GameplayEffects|RemoveActiveEffectswithAppliedTags (Ability|GetAbilitySystemComponent _fortplayerpawn) (Utilities|Struct|BreakInheritedTagContainer _inheritableownedtagscontainer))
      (EventDispatchers|UnbindEventfromOnPawnLanded _fortplayerpawn (delegate Custom|OnPawnLanded_Event_0))
      (Ability|EndAbility))))

(event Custom|ForceEndAbilityIgnoreLanded
  (bind (_period _bexecuteperiodiceffectonapplication _periodicinhibitionpolicy _modifiers _executions _conditionalgameplayeffects _overfloweffects _bdenyoverflowapplication _bclearstackonoverflow _prematureexpirationeffectclasses _routineexpirationeffectclasses _brequiremodifiersuccesstotriggercues _gameplaycues _inheritablegameplayeffecttags _inheritableownedtagscontainer) (Class|GetClassDefaults (Variables|Launched|Settings|GameplayEffect|GetGEAppliedRedeployTemplate)))
  ;[1 1 1] Redeploy Logic
  (bind _fortplayerpawn (Utilities|GetValidFortPlayerPawnfromInfo)
    (:Valid
      (GameplayEffects|RemoveActiveEffectswithAppliedTags (Ability|GetAbilitySystemComponent _fortplayerpawn) (Utilities|Struct|BreakInheritedTagContainer _inheritableownedtagscontainer))
      (EventDispatchers|UnbindEventfromOnPawnLanded _fortplayerpawn (delegate Custom|OnPawnLanded_Event_0))
      (Ability|EndAbility))))

(event Custom|ListenForWalking
  ;[1 1 1] Listen for if player goes into walking state, end ability if so
  (Utilities|GetValidFortPlayerPawnfromInfo
    (:Valid
      (Ability|Tasks|WaitMovementModeChange "MOVE_Walking"
        (:OnChange
          (Ability|EndAbility))))))

(event Custom|ApplyVisualFeedback
  (params (bEndAbility :type Boolean))
  ;[1 1 1] Apply visual feedback
  (Utilities|FlowControl|Sequence
    (:then_0
      (bind _fortplayercontroller (Utilities|GetvalidFortPlayerControllerfromInfo)
        (:Valid
          (Game|Feedback|LocalPlayForceFeedback _fortplayercontroller (Variables|Launched|Settings|VisualFeedback|GetForcedFeedback) false "DirectionalLaunch")
          (Utilities|IsServer
            (:true
              (Game|Feedback|ClientStartCameraShake :self _fortplayercontroller :Shake (Variables|Launched|Settings|VisualFeedback|GetPlayerLaunchedCameraShakeTemplate) :Scale 1.0 :PlaySpace "CameraLocal")
              (Game|Feedback|ClientSpawnGenericCameraLensEffect _fortplayercontroller (LensEffect|GetInterfaceClass (Variables|Launched|Settings|VisualFeedback|GetCameraLensClass_PlayerLaunched))))))
        (:Invalid
          (Ability|EndAbility))))
    (:then_1
      (if bEndAbility
        (Ability|EndAbility)))))

(event Ability|EventOnEndAbility
  (params (bWasCancelled :type Boolean))
  ;[1 1 1] Ability end
  (bind (_athenavehicle _fortpawn) (Utilities|GetValidVehicleorPawnfromInfo)
    (:Vehicle
      (GameplayEffects|RemoveActiveGameplayEffect (Ability|GetAbilitySystemComponent _athenavehicle) (Variables|Launched|GameplayDynamic|GetGELaunchedAppliedHandle_Vehicle) -1))
    (:Player
      (bind _asfort_player_pawn (Utilities|Casting|CastToFortPlayerPawn _fortpawn)
        (:then
          (GameplayEffects|RemoveActiveGameplayEffect (Ability|GetAbilitySystemComponent _asfort_player_pawn) (Variables|Launched|GameplayDynamic|GetGELaunchedAppliedHandle) -1)
          (GameplayEffects|RemoveActiveGameplayEffect (Ability|GetAbilitySystemComponent _asfort_player_pawn) (Variables|Launched|GameplayDynamic|GetGERedeployHandle) -1))
        (:CastFailed
          _)))
    (:Neither
      _))
  (bind _wildlife_actor (Utilities|IsActorWildlife)
    (:Valid
      (GameplayEffects|RemoveActiveGameplayEffect (Ability|GetAbilitySystemComponent _wildlife_actor) (Variables|Launched|GameplayDynamic|GetGELaunchedAppliedHandle) -1))))

(fn AttemptLaunchPhysicsObject
  (params (LaunchedPhysicsObject :type Actor%20Object%20Reference))
  (bind _level (CallFunction|GetBaseVelocityLevel))
  (bind _launchedphysicsobject (Variables|GetLaunchedPhysicsObject))
  (bind (_directionallaunchparent _velocity) (Utilities|GetValidSourceObjectasDirectionalLaunchParent)
    (:Valid
      (bind _asfort_physics_pawn_object (Utilities|Casting|CastToFortPhysicsPawnObject (Variables|GetLaunchedPhysicsObject))
        (:then
          (Physics|SetAllPhysicsLinearVelocity (Variables|/Script/FortniteGame.FortPhysicsPawnObject|GetPhysicsMeshComponent _asfort_physics_pawn_object) (* (Math|Vector|SelectVector (CallFunction|GetLaunchTowardLaunchTargetAngle _launchedphysicsobject _directionallaunchparent (Variables|/Game/Athena/Environments/Blueprints/EnvironmentalLaunch/BGA_DirectionalLaunchParent.BGA_DirectionalLaunchParent_C|GetLaunchDirectionOfUserIsHorizontalOnly _directionallaunchparent)) (Math|Rotator|GetRotationXVector (Transformation|GetWorldRotation (Variables|/Game/Athena/Environments/Blueprints/EnvironmentalLaunch/BGA_DirectionalLaunchParent.BGA_DirectionalLaunchParent_C|GetLaunchDirection _directionallaunchparent))) (Variables|/Game/Athena/Environments/Blueprints/EnvironmentalLaunch/BGA_DirectionalLaunchParent.BGA_DirectionalLaunchParent_C|GetLaunchesIndirectionOfUser _directionallaunchparent)) (ScalableFloat|GetValueatLevel _velocity _level)) false))
        (:CastFailed
          (FortPhysicsSimulationLibrary|SetPhysicsLinearVelocity (Variables|GetLaunchedPhysicsObject) (* (Math|Vector|SelectVector (CallFunction|GetLaunchTowardLaunchTargetAngle _launchedphysicsobject _directionallaunchparent (Variables|/Game/Athena/Environments/Blueprints/EnvironmentalLaunch/BGA_DirectionalLaunchParent.BGA_DirectionalLaunchParent_C|GetLaunchDirectionOfUserIsHorizontalOnly _directionallaunchparent)) (Math|Rotator|GetRotationXVector (Transformation|GetWorldRotation (Variables|/Game/Athena/Environments/Blueprints/EnvironmentalLaunch/BGA_DirectionalLaunchParent.BGA_DirectionalLaunchParent_C|GetLaunchDirection _directionallaunchparent))) (Variables|/Game/Athena/Environments/Blueprints/EnvironmentalLaunch/BGA_DirectionalLaunchParent.BGA_DirectionalLaunchParent_C|GetLaunchesIndirectionOfUser _directionallaunchparent)) (ScalableFloat|GetValueatLevel _velocity _level)) false "None" 0 "(MaxSimSizeAllowed=Large,bRejectEmbeddedInLandscape=False)"))) ; FortPhysicsPawnObject actors require SetAllPhysicsLinearVelocity
      (CallFunction|ApplyGameplayCuePlayerLaunch) ; The devices use a cue here that accounts for physics objects
      (Ability|EndAbility))
    (:Invalid
      (Ability|EndAbility))))

(fn GetLaunchTowardLaunchTargetAngle
  (params (Target :type Actor%20Object%20Reference) (Launcher :type Actor%20Object%20Reference) (HorizontalOnly :type Boolean))
  (bind (_returnvalue_x _returnvalue_y _returnvalue_z) (Math|Vector|BreakVector (Transformation|GetActorLocation Target)))
  (bind (_returnvalue_x_2 _returnvalue_y_2 _returnvalue_z_2) (Math|Vector|BreakVector (Transformation|GetActorLocation Launcher)))
  ;[1 1 1] Find the angle to launche the target to launch them away from the launcher.
  (return (Math|Vector|Normalize (- (Math|Vector|MakeVector _returnvalue_x _returnvalue_y (Math|Float|SelectFloat 0.0 _returnvalue_z HorizontalOnly)) (Math|Vector|MakeVector _returnvalue_x_2 _returnvalue_y_2 (Math|Float|SelectFloat 0.0 _returnvalue_z_2 HorizontalOnly))))))

(fn GetBaseVelocityLevel
  (return 0.0))

(fn GetFortActorFromInfo
  (bind _returnvalue (Ability|GetAvatarActorfromActorInfo))
  (Utilities|IsValid _returnvalue
    (:Is%20Valid
      (return _returnvalue true))
    (:Is%20Not%20Valid
      (return 0 false))))

(fn GetSourceObject
  (return (Ability|EffectContext|GetSourceObject (Ability|GetContextfromOwner))))

(fn GetVehicleVelocityModifer
  (return (ScalableFloat|GetValueatLevel (Variables|Launched|Settings|Hotfix|GetVehicleVelocityModifier))))

(fn IsPlayerBeingLaunched
  (params (Actor :type Actor%20Object%20Reference))
  (if (not (CallFunction|IsVehicleBeingLaunched Actor))
    (bind _asfort_pawn (Utilities|Casting|CastToFortPawn Actor)
      (:then
        (return _asfort_pawn true))
      (:CastFailed
        _)))
  (return 0 false))

(fn IsVehicleBeingLaunched
  (params (Actor :type Actor%20Object%20Reference))
  (bind _asfort_pawn (Utilities|Casting|CastToFortPawn Actor)
    (:then
      (bind _vehicle_occupied (Utilities|IsinVehicle _asfort_pawn)
        (:true
          (bind _asfort_athena_vehicle (Utilities|Casting|CastToFortAthenaVehicle _vehicle_occupied)
            (:then
              (return true _asfort_athena_vehicle))
            (:CastFailed
              (return false 0))))
        (:false
          _)))
    (:CastFailed
      _))
  (bind _asfort_athena_vehicle_2 (Utilities|Casting|CastToFortAthenaVehicle Actor)
    (:then
      (return true _asfort_athena_vehicle_2))
    (:CastFailed
      (return false 0))))
